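# Subsurface, Nanite-enabled material for birthday party balloon props, with animated sway.
# Vertex color R and G blend three color parameters, tinted by a diffuse texture, overlaid with a per-color-maskable UV1 decal, plus Fresnel rim.
# World position offset sums a slow actor-pivot wobble and Pivot Painter per-element wind rotation with random phase, masked by vertex G.
# Attributes pass through MF_CreativeMoveTool; some trailing position nodes are unconnected.
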Material /Game/Environments/Props/BirthdayBalloons/Materials/M_PartyBalloons
graph {
 properties {
  bUseMaterialAttributes=true
  bUsedWithNanite=true
  pixelDepthOffsetMode=PDOM_Legacy
  shadingModel=MSM_Subsurface
  usedShadingModels=Subsurface
 }
 nodes [
  n1 = Multiply(
   A=Fresnel(ExponentIn=ScalarParameter() {defaultValue=2.3149, parameterName=FresnelPower}) {baseReflectFraction=0}
   B=VectorParameter() {parameterName=FresnelColor, defaultValue=<rgba 0.4115 0.4115 0.4115 1>}.RGB
  )
  n2 = VertexColor()
  n3 = LinearInterpolate(
   A=VectorParameter() {group=Colors, parameterName=Color_1, defaultValue=<rgba 0 1 0.0445 1>}.RGB
   B=VectorParameter() {group=Colors, parameterName=Color_2, defaultValue=<rgba 1 0.01 0 1>}.RGB
   Alpha=@n2.R
  )
  n4 = VectorParameter() {group=Colors, parameterName=Color_3, defaultValue=<rgba 0 0.235 0.9635 1>}
  n5 = MaterialFunctionCall(
   UVs=TextureCoordinate() {coordinateIndex=1}
   Texture Scale=AppendVector(
    A=ScalarParameter() {defaultValue=1, group=Decal, parameterName=DecalScale_X}
    B=ScalarParameter() {defaultValue=1, group=Decal, parameterName=DecalScale_Y}
   )
  ) {materialFunction=&ScaleUVsByCenter}
  n6 = TextureSampleParameter2D(UVs=@n5.UVs) {group=Decal, parameterName=DecalTexture, texture=/Game/Environments/Props/BirthdayBalloons/Textures/T_NoDecal}
  n7 = ScalarParameter() {group=Colors, parameterName=Color1DecalMaskOut}
  n8 = ScalarParameter() {group=Colors, parameterName=Color2DecalMaskOut}
  n9 = OneMinus(None=LinearInterpolate(A=LinearInterpolate(A=@n7, B=@n8, Alpha=@n2.R), B=ScalarParameter() {group=Colors, parameterName=Color3DecalMaskOut}, Alpha=@n2.G))
  n10 = LinearInterpolate(
   A=Multiply(
    A=LinearInterpolate(A=@n3, B=@n4.RGB, Alpha=@n2.G)
    B=TextureSample() {texture=/Game/Environments/Props/BirthdayBalloons/Textures/T_PartyBalloon_D}.RGB
   )
   B=@n6.RGB
   Alpha=Multiply(A=Multiply(A=@n6.A, B=@n5.0-1 mask), B=Saturate(None=@n9))
  )
  n11 = Multiply(A=Time(), B=ScalarParameter() {defaultValue=0.1714, parameterName=SpeedBallon_2})
  n12 = Sine(None=@n11)
  n13 = AppendVector(A=Sine(None=@n11), B=Cosine(None=@n11))
  n14 = Multiply(A=Time(), B=ScalarParameter() {defaultValue=0.0892, parameterName=OverallRotationSpeed})
  n15 = WorldPosition() {worldPositionShaderOffset=WPT_ExcludeAllShaderOffsets}
  n16 = RotateAboutAxis(
   NormalizedRotationAxis=Normalize(VectorInput=AppendVector(A=@n12, B=@n13))
   RotationAngle=Multiply(A=Sine(None=@n14), B=ScalarParameter() {defaultValue=0.01, parameterName=OverallRotation})
   PivotPoint=Add(A=Constant3Vector(), B=ActorPositionWS())
   Position=@n15.XYZ
  )
  n17 = Transform(None=VectorParameter() {parameterName=RotaionAxis, defaultValue=<rgba 0 0 1 1>}.RGB) {transformSourceType=TRANSFORMSOURCE_Local}
  n18 = Time()
  n19 = MaterialFunctionCall() {materialFunction=/Engine/Functions/Engine_MaterialFunctions02/PivotPainter/PivotPainter_PerObjectData}
  n20 = Add(A=@n18, B=Multiply(A=@n19.Random Value Per Element, B=ScalarParameter() {defaultValue=1.5613, parameterName=OffsetVariation}))
  n21 = ScalarParameter() {defaultValue=0.444, parameterName=WindRotation_Speed_1}
  n22 = Add(A=Multiply(A=@n19.Random Value Per Element, B=ScalarParameter() {defaultValue=2.2338, parameterName=OffsetVariation_1}), B=@n18)
  n23 = ScalarParameter() {defaultValue=0.469, parameterName=WindRotation_Speed}
  n24 = Add(A=Sine(None=Multiply(A=@n20, B=@n21)), B=Sine(None=Multiply(A=@n22, B=@n23)))
  n25 = ScalarParameter() {defaultValue=0.02, parameterName=WindRotation_strength}
  n26 = Multiply(
   A=RotateAboutAxis(
    NormalizedRotationAxis=Normalize(VectorInput=@n17)
    RotationAngle=Multiply(A=@n24, B=@n25)
    PivotPoint=@n19.Pivot Position
    Position=@n15.XYZ
   )
   B=VertexColor().G
  )
  n27 = ScalarParameter() {defaultValue=0.4857, parameterName=SSSBrightness}
  MaterialFunctionCall(MA In=MakeMaterialAttributes!(
    BaseColor=Add(A=@n1, B=@n10)
    Metallic
    Specular
    Roughness=TextureSample() {texture=/Game/Environments/Props/BirthdayBalloons/Textures/T_PartyBalloon_M}.B
    Anisotropy
    EmissiveColor
    Opacity=ScalarParameter() {defaultValue=0.5, parameterName=SSSDistance}
    OpacityMask
    Normal=TextureSample() {samplerType=SAMPLERTYPE_Normal, texture=/Game/Environments/Props/BirthdayBalloons/Textures/T_PartyBalloon_N}.RGB
    Tangent
    WorldPositionOffset=Add(A=@n16, B=@n26)
    SubsurfaceColor=Multiply(A=@n10, B=@n27)
    ClearCoat
    ClearCoatRoughness
    AmbientOcclusion
    Refraction
    CustomizedUVs_0
    CustomizedUVs_1
    CustomizedUVs_2
    CustomizedUVs_3
    CustomizedUVs_4
    CustomizedUVs_5
    CustomizedUVs_6
    CustomizedUVs_7
    PixelDepthOffset
    ShadingModel
    Displacement
   )) {materialFunction=&MF_CreativeMoveTool}
  n28 = AppendVector(A=TextureCoordinate() {coordinateIndex=2}, B=ComponentMask(None=TextureCoordinate() {coordinateIndex=3}) {r=true})
  n29 = Constant3Vector() {constant=<rgba 0 -1 0 0>}
  Normalize(VectorInput=TransformPosition(None=Add(A=@n28, B=@n29)) {transformType=TRANSFORMPOSSOURCE_World})
  VectorParameter() {parameterName=RotationCenter, defaultValue=<rgba 0 0 0 1>}
  WorldPosition()
 ]
}
